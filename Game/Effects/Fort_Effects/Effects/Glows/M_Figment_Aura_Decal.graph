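# Additive unlit Niagara sprite material for the Figment aura glow, faking a decal as a band where the sprite nears scene geometry.
# Flipbook tile UVs, warped by two screen-aligned panning hex-noise samples, drive a glitch texture shaped by dynamic parameters and Fresnel.
# That pattern multiplies a scene-minus-pixel depth band and particle color into emissive adjustments.
# IS_NoDamage_ON drops the flipbook UVs and changes pan speeds.
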Material /Game/Effects/Fort_Effects/Effects/Glows/M_Figment_Aura_Decal
graph {
 properties {
  bAutomaticallySetUsageInEditor=false
  bUsedWithNiagaraSprites=true
  blendMode=BLEND_Additive
  shadingModel=MSM_Unlit
  usedShadingModels=Unlit
 }
 nodes [
  Constant() {r=1}
  n1 = TextureCoordinate() {vTiling=0}
  n2 = Panner(
   Coordinate=Multiply(A=ComponentMask(None=@n1) {r=true}, B=ScalarParameter() {defaultValue=1, parameterName=PatternTileU})
   Time=ScalarParameter() {parameterName=Pattern_PanU}
  ) {speedX=0.5}
  n3 = TextureCoordinate() {uTiling=0}
  n4 = Panner(
   Coordinate=Multiply(A=ComponentMask(None=@n3) {g=true}, B=ScalarParameter() {defaultValue=1, parameterName=PatternTileV})
   Time=ScalarParameter() {parameterName=Pattern_PanV}
  ) {speedY=0.5}
  n5 = Constant() {r=1}
  n6 = ScalarParameter() {defaultValue=100, parameterName=Columns}
  n7 = ScalarParameter() {defaultValue=100, parameterName=Rows}
  n8 = Multiply(
   A=AppendVector(A=ComponentMask(None=@n2) {r=true}, B=ComponentMask(None=@n4) {g=true})
   B=AppendVector(A=Divide(A=@n5, B=@n6), B=Divide(A=@n5, B=@n7))
  )
  n9 = Multiply(A=Add(A=Time(), B=ParticleRandom()), B=ScalarParameter() {defaultValue=16, parameterName=FB_Speed})
  n10 = Floor(None=@n9)
  n11 = Divide(A=@n9, B=@n7)
  n12 = AppendVector(A=Divide(A=Fmod(A=@n10, B=@n7), B=@n7), B=Divide(A=Floor(None=@n11), B=@n6))
  n13 = MaterialFunctionCall() {materialFunction=&ScreenAlignedUVs}
  n14 = VectorParameter() {parameterName=Hex UVs, defaultValue=<rgba 9 5 5.5 3>}
  n15 = Time()
  n16 = Panner(
   Coordinate=Multiply(A=@n13."X 100%, Y 100%", B=AppendVector(A=@n14.R, B=@n14.R)) {constB=4}
   Time=@n15
   Speed=StaticSwitchParameter(True=Constant2Vector() {g=0.4}, False=Constant2Vector() {g=-0.8, r=0.15}) {parameterName=IS_NoDamage_ON}
  ) {speedY=-0.8}
  n17 = Panner(
   Coordinate=Multiply(A=@n13."X 100%, Y 100%", B=AppendVector(A=@n14.B, B=@n14.B)) {constB=3}
   Time=@n15
   Speed=StaticSwitchParameter(True=Constant2Vector() {g=0.2}, False=Constant2Vector() {g=-0.3, r=0.22}) {parameterName=IS_NoDamage_ON}
  ) {speedY=-0.3}
  n18 = Add(
   A=StaticSwitchParameter(True=Constant(), False=Add(A=@n8, B=@n12)) {parameterName=IS_NoDamage_ON}
   B=AppendVector(
    A=TextureSample(UVs=@n16) {texture=/Game/Effects/Fort_Effects/Textures/Tile/T_Hex_Noise}.R
    B=TextureSample(UVs=@n17) {texture=&T_Hex_Noise}.G
   )
  )
  n19 = TextureSampleParameter2D(UVs=StaticSwitchParameter(True=Panner(Coordinate=@n18) {speedX=-0.05, speedY=0.01}, False=Panner(Coordinate=@n18) {speedX=-0.25, speedY=0.01}) {parameterName=IS_NoDamage_ON}) {parameterName=Texture1, texture=/Game/Effects/Fort_Effects/Textures/Glitches/T_VCVFX_Boost_Glitch_Pixel_Alpha}
  n20 = DynamicParameter() {defaultValue=<rgba 20 1 1 1>, paramNames [Tex Intensity, Tex Pow, Param3, Param4]}
  n21 = Multiply(
   A=Fresnel(
    ExponentIn=ScalarParameter() {defaultValue=0.2, parameterName=Expo}
    BaseReflectFractionIn=ScalarParameter() {defaultValue=-0.01, parameterName=Base reflact}
   )
   B=Multiply(A=Power(Base=@n19.R, Exp=@n20.Tex Pow), B=@n20.Tex Intensity) {constB=4}
  )
  n22 = Subtract(A=SceneDepth().R, B=PixelDepth().R)
  n23 = Multiply(A=ScalarParameter() {defaultValue=0.01, parameterName=m}, B=@n22)
  n24 = Multiply(
   A=Power(Base=Saturate(None=@n23), Exp=ScalarParameter() {defaultValue=0.5, parameterName=Pow})
   B=ScalarParameter() {defaultValue=5, parameterName=aa}
  )
  n25 = ScalarParameter() {defaultValue=0.011, parameterName=Dist}
  n26 = OneMinus(None=Saturate(None=Multiply(A=@n22, B=@n25)))
  n27 = Power(Base=@n26, Exp=ScalarParameter() {defaultValue=1.9, parameterName=Pow2})
  n28 = ScalarParameter() {defaultValue=6, parameterName=a}
  n29 = Subtract(A=Multiply(A=Saturate(None=@n24), B=@n26), B=Saturate(None=Multiply(A=@n27, B=@n28)))
  n30 = ParticleColor()
  MaterialFunctionCall(Emissive=Multiply(A=NamedRerouteDeclaration(None=@n21) {name=Texture Result, nodeColor=<rgba 1 0 0.2812 1>}, B=Multiply(A=@n29, B=@n30.RGB))) {materialFunction=&MF_FX_EmissiveAdjustments}
  n31 = TextureCoordinate()
  n32 = Constant() {r=0.5}
  n33 = Multiply(
   A=ScalarParameter() {defaultValue=0.5, group=Opacity, parameterName=SphereMaskRadius}
   B=DynamicParameter() {parameterIndex=3, paramNames [MaskRadius_Dyn, Param2, Param3, Param4]}.MaskRadius_Dyn
  )
  Multiply(B=Power(
    Base=SphereMask(A=@n31, B=@n32, Radius=@n33) {hardnessPercent=0}
    Exp=ScalarParameter() {defaultValue=2, group=Opacity, parameterName=SphereMaskPower, sortPriority=5}
   ))
  WorldPosition()
 ]
}
