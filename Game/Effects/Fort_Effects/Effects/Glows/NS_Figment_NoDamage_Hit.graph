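# Niagara hit effect, named for no-damage hits on a Figment, that briefly flashes a blue glow over a character's body.
# The active Bright_Flash_Decal001 emitter bursts 13 partitioned particles, each pinned every update to a limb, head or foot bone with per-part offsets and scales.
# Sprite and mesh renderers draw them, with curves fading color and driving material parameters, scaled by time of day.
# Glow and the original Bright_Flash_Decal emitter are disabled.
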NiagaraSystem /Game/Effects/Fort_Effects/Effects/Glows/NS_Figment_NoDamage_Hit
dependencies {
 usedDataInterfaces [&NiagaraDataInterfaceCurve, &NiagaraDataInterfaceArrayFloat, &NiagaraDataInterfaceArrayFloat3]
 usedDynamicInputs [
  &SetBoolByFloatComparison
  &FN_Select_Float_from_Time_Of_Day
  /Game/Weapons/FORT_Melee/Pickaxe_GateHound/FX/VariantManager/VariantVectorSelector
  &FN_Select_Vector_from_Time_Of_Day
 ]
 usedModules [
  &self:SystemSpawnScript.SystemScriptSource.SystemScriptGraph.NiagaraNodeAssignment_0.SetVariables_#1
  &self:SystemSpawnScript.SystemScriptSource.SystemScriptGraph.NiagaraNodeAssignment_2.SetVariables_#2
  &SystemState
  &Timeline
  /Game/Weapons/FORT_Melee/Pickaxe_GateHound/FX/VariantManager/FN_VariantManager
  &self:SystemSpawnScript.SystemScriptSource.SystemScriptGraph.NiagaraNodeAssignment_7.SetVariables_#3
  &self:SystemSpawnScript.SystemScriptSource.SystemScriptGraph.NiagaraNodeAssignment_8.SetVariables_#4
 ]
}
emitters {
 Bright_Flash_Decal {
  bEnabled=false
  emitterUpdateScript [
   &EmitterState(
    Inactive Response=::NewEnumerator1
    Life Cycle Mode=::NewEnumerator1
    Loop Behavior=::NewEnumerator1
    Loop Delay=0.1
    Loop Duration=0.5
    UseLoopDelay=false
   )
   &SpawnRate(SpawnRate=30, Use Spawn Probability=false) {enabled=false}
   &SpawnBurst_Instantaneous(Spawn Count=13, Spawn Time=0.1, Use Spawn Probability=false)
   /SetVariables(
    Emitter.Outline_size:NiagaraFloat=&FloatFromCurve(
     FloatCurve=DataInterface() {propertyValues {Curve {keys [
         {arriveTangent=0.1954, interpMode=RCIM_Cubic, leaveTangent=0.1954, tangentMode=RCTM_User, time=0, value=0.85}
         {arriveTangent=-9.4937, interpMode=RCIM_Cubic, leaveTangent=-9.4937, tangentMode=RCTM_User, time=0.3, value=0}
        ]}}}
     CurveIndex=$Emitter.Age
     Scale Curve=1.2
    )
    Emitter.Outline_size2:NiagaraFloat=&FloatFromCurve(
     FloatCurve=DataInterface() {propertyValues {Curve {keys [
         {arriveTangent=0.1954, interpMode=RCIM_Cubic, leaveTangent=0.1954, tangentMode=RCTM_User, time=0, value=0.85}
         {arriveTangent=1.4805, interpMode=RCIM_Cubic, leaveTangent=1.4805, tangentMode=RCTM_User, time=0.3, value=0}
        ]}}}
     CurveIndex=$Emitter.Age
     Scale Curve=15
    )
   ) {isSetParameters=true, module=SetVariables_#5}
   &SpawnBurst_Instantaneous(Spawn Count=13, Spawn Time=0.5, Use Spawn Probability=false, UseLoopCountLimit=false) {module=SpawnBurst_Instantaneous001}
  ]
  particleSpawnScript [
   &PartitionParticles(Number Of Partitions=13)
   /Niagara/Modules/Spawn/Initialization/InitializeParticle(
    Color=<rgba 0 1.5288 5 1>
    Lifetime=&RandomRangeFloat(Minimum=0.7, Maximum=0.8)
    Mass=&RandomRangeFloat(Minimum=0.6, Maximum=1.2)
    Mesh Scale=&SelectVectorFromArray(
     Direct Array Index=$Particles.PartitionParticles.Partition
     Vector Selection Array=DataInterface() {propertyValues {FloatData [
        <0.3 0.3 0.3>
        <0.3 0.3 0.3>
        <0.3 0.3 0.3>
        <0.3 0.3 0.3>
        <0.3 0.3 0.3>
        <0.3 0.3 0.3>
        <0.3 0.3 0.3>
        <0.3 0.3 0.3>
        <0.3 0.3 0.5>
        <0.3 0.3 0.5>
        <0.3 0.3 0.4>
        <0.35 0.35 0.35>
        <0.35 0.35 0.35>
       ]}}
    )
    Module.Write Color=true
    Module.Write Lifetime=true
    Module.Write Mass=false
    Module.Write Material Random=true
    Module.Write Position=false
    Module.Write Scale=true
    Module.Write SpriteRotation=true
    Module.Write SpriteSize=true
    Sprite Rotation=&RandomRangeFloat(Minimum=0, Maximum=360)
    Sprite Size=&Vector2DFromFloat(Value=&Multiply_Float(
      A=40
      B=&SelectFloatFromArray(
       Direct Array Index=$Particles.PartitionParticles.Partition
       Float Selection Array=DataInterface() {propertyValues {FloatData [1,1,1,1,1,1,1,1,1,1,1,0,0]}}
      )
     ))
   )
   &CameraOffset(Camera Offset Amount=100)
   /SetVariables(Particles.CamOfs_Mesh=15) {isSetParameters=true, module=SetVariables_#6}
  ]
  particleUpdateScript [
   &ParticleState()
   &SkeletalMeshLocation(
    Bone / Socket Index=$Particles.PartitionParticles.Partition
    Bone Sampling Mode=::NewEnumerator4
    MaskSpawnGroup=false
    Orientation Sampling=::NewEnumerator1
    Sampled Position Offset=&SelectVectorFromArray(
     Direct Array Index=$Particles.PartitionParticles.Partition
     Vector Selection Array=DataInterface() {propertyValues {FloatData [<0 0 0>, <1 1 1>, <0 0 0>, <0 0 0>, <0 0 0>, <0 0 0>, <0 0 0>, <0 0 0>, <0 0 0>, <0 0 0>, <10 0 0>, <0 -7 0>, <0 7 0>]}}
    )
    Skeletal Mesh=DataInterface() {propertyValues {
      DefaultMesh=null
      ExcludeBoneName=root
      OptionalComponentClass=null
      PreviewMesh=/Game/Characters/Player/Female/Medium/Bodies/F_MED_Default_Soldier/Meshes/F_MED_Default_Soldier
      SkinningMode=SkinOnTheFly
      SoftSourceActor=null
      SourceMode=Default
      UvSetIndex=0
      WholeMeshLOD=-1
      bExcludeBone=false
      bReadDeformedGeometry=true
      bRequireCurrentFrameData=false
      ComponentTags []
      FilteredBones [upperarm_l,upperarm_r,lowerarm_l,lowerarm_r,hand_l,hand_r,thigh_l,thigh_r,calf_l,calf_r,head,foot_r,foot_l]
      FilteredSockets []
      MeshUserParameter {parameter {name=None, typeDefHandle {registeredTypeIndex=1}, varData []}}
      SamplingRegions []
     }}
   )
   &ScaleColor(
    Scale Alpha=&FloatFromCurve(FloatCurve=DataInterface() {propertyValues {Curve {keys [
         {arriveTangent=21.7774, interpMode=RCIM_Cubic, leaveTangent=21.7774, tangentMode=RCTM_User, time=0, value=0}
         {arriveTangent=-0.2619, interpMode=RCIM_Cubic, leaveTangent=-0.2619, tangentMode=RCTM_User, time=0.3, value=3}
         {arriveTangent=-5.2661, interpMode=RCIM_Cubic, leaveTangent=-5.2661, tangentMode=RCTM_User, time=1, value=0}
        ]}}})
    Scale RGB=&Multiply_VectorByFloat(
     Vector=<2 2 2>
     Float=&FloatFromCurve(
      FloatCurve=DataInterface() {propertyValues {Curve {keys [
          {arriveTangent=21.7774, interpMode=RCIM_Cubic, leaveTangent=21.7774, tangentMode=RCTM_User, time=0, value=0}
          {arriveTangent=-0.2619, interpMode=RCIM_Cubic, leaveTangent=-0.2619, tangentMode=RCTM_User, time=0.3, value=3}
          {arriveTangent=-5.2661, interpMode=RCIM_Cubic, leaveTangent=-5.2661, tangentMode=RCTM_User, time=1, value=0}
         ]}}}
      Scale Curve=&FNTimeOfDayScalar(Night Scale=0.25, Morning Scale=0.7, Evening Scale=0.7)
     )
    )
    ScaleA=false
    ScaleRGB=true
   )
   &DynamicMaterialParameters(
    Index 0 Param 1=&FloatFromCurve(
     FloatCurve=DataInterface() {propertyValues {Curve {
        defaultValue=0
        keys [
         {tangentMode=RCTM_User, time=0, value=0.55}
         {arriveTangent=-0.0405, interpMode=RCIM_Cubic, leaveTangent=-0.0405, tangentMode=RCTM_User, time=0.2, value=0.55}
         {arriveTangent=-8.6221, interpMode=RCIM_Cubic, leaveTangent=-8.6221, tangentMode=RCTM_User, time=0.3, value=0}
        ]
       }}}
     CurveIndex=$Emitter.Age
     Scale Curve=&FNTimeOfDayScalar(Night Scale=40, Morning Scale=40, Day Scale=20, Evening Scale=40)
    )
    Index 0 Param 2=&FloatFromCurve(FloatCurve=DataInterface() {propertyValues {Curve {keys [<at 0 1>,<at 1 0>]}}}, Scale Curve=0.5)
    Module.Index0Float0Write=true
    Module.Index0Float1Write=true
    Module.Index0Float2Write=false
    Module.Index0Float3Write=false
   ) {module=DynamicMaterialParameters001}
  ]
  rendererClasses [&NiagaraSpriteRendererProperties, &NiagaraMeshRendererProperties]
  renderers [
   {rendererClass=&NiagaraSpriteRendererProperties}
   {rendererClass=&NiagaraMeshRendererProperties, rendererIndex=1}
  ]
 }
 Bright_Flash_Decal001 {
  emitterUpdateScript [
   &EmitterState(
    Inactive Response=::NewEnumerator1
    Life Cycle Mode=::NewEnumerator1
    Loop Behavior=::NewEnumerator1
    Loop Delay=0.1
    Loop Duration=0.5
    UseLoopDelay=false
   )
   &SpawnRate(SpawnRate=30, Use Spawn Probability=false) {enabled=false}
   &SpawnBurst_Instantaneous(Spawn Count=13, Spawn Time=0.1, Use Spawn Probability=false)
   /SetVariables(
    Emitter.Outline_size:NiagaraFloat=&FloatFromCurve(
     FloatCurve=DataInterface() {propertyValues {Curve {keys [
         {arriveTangent=0.1954, interpMode=RCIM_Cubic, leaveTangent=0.1954, tangentMode=RCTM_User, time=0, value=0.85}
         {arriveTangent=-9.4937, interpMode=RCIM_Cubic, leaveTangent=-9.4937, tangentMode=RCTM_User, time=0.3, value=0}
        ]}}}
     CurveIndex=$Emitter.Age
     Scale Curve=1.2
    )
    Emitter.Outline_size2:NiagaraFloat=&FloatFromCurve(
     FloatCurve=DataInterface() {propertyValues {Curve {keys [
         {arriveTangent=0.1954, interpMode=RCIM_Cubic, leaveTangent=0.1954, tangentMode=RCTM_User, time=0, value=0.85}
         {arriveTangent=1.4805, interpMode=RCIM_Cubic, leaveTangent=1.4805, tangentMode=RCTM_User, time=0.3, value=0}
        ]}}}
     CurveIndex=$Emitter.Age
     Scale Curve=15
    )
   ) {enabled=false, isSetParameters=true, module=SetVariables_#5}
   /SetVariables(Emitter.Thickness:NiagaraFloat=&FloatFromCurve(
     FloatCurve=DataInterface() {propertyValues {Curve {keys [
         <at 0 0.2 RCIM_Cubic>
         {arriveTangent=0.425, interpMode=RCIM_Cubic, leaveTangent=0.425, time=0.2, value=1.375}
         {arriveTangent=-1.889, interpMode=RCIM_Cubic, leaveTangent=-1.889, tangentMode=RCTM_User, time=1, value=0.625}
         {interpMode=RCIM_Cubic, leaveTangent=-0.6944, time=1.5, value=0}
        ]}}}
     CurveIndex=$Emitter.Age
     Scale Curve=2
    )) {isSetParameters=true}
  ]
  particleSpawnScript [
   &PartitionParticles(Number Of Partitions=13)
   &InitializeParticle(
    Color=<rgba 0 1.5288 5 1>
    Lifetime=&RandomRangeFloat(Minimum=0.5, Maximum=0.5)
    Mass=&RandomRangeFloat(Minimum=0.6, Maximum=1.2)
    Mesh Scale=&SelectVectorFromArray(
     Direct Array Index=$Particles.PartitionParticles.Partition
     Vector Selection Array=DataInterface() {propertyValues {FloatData [
        <0.3 0.3 0.3>
        <0.3 0.3 0.3>
        <0.3 0.3 0.3>
        <0.3 0.3 0.3>
        <0.3 0.3 0.3>
        <0.3 0.3 0.3>
        <0.3 0.3 0.3>
        <0.3 0.3 0.3>
        <0.3 0.3 0.5>
        <0.3 0.3 0.5>
        <0.3 0.3 0.4>
        <0.35 0.35 0.35>
        <0.35 0.35 0.35>
       ]}}
    )
    Module.Write Color=true
    Module.Write Lifetime=true
    Module.Write Mass=false
    Module.Write Material Random=true
    Module.Write Position=false
    Module.Write Scale=true
    Module.Write SpriteRotation=true
    Module.Write SpriteSize=true
    Sprite Rotation=&RandomRangeFloat(Minimum=0, Maximum=360)
    Sprite Size=&Vector2DFromFloat(Value=&Multiply_Float(
      A=40
      B=&SelectFloatFromArray(
       Direct Array Index=$Particles.PartitionParticles.Partition
       Float Selection Array=DataInterface() {propertyValues {FloatData [1,1,1,1,0.9,0.9,1,1,1,1,1,0,0]}}
      )
     ))
   )
   &CameraOffset(Camera Offset Amount=85)
   /SetVariables(Particles.CamOfs_Mesh=15) {isSetParameters=true, module=SetVariables_#6}
  ]
  particleUpdateScript [
   &ParticleState()
   &SkeletalMeshLocation(
    Bone / Socket Index=$Particles.PartitionParticles.Partition
    Bone Sampling Mode=::NewEnumerator4
    MaskSpawnGroup=false
    Orientation Sampling=::NewEnumerator1
    Sampled Position Offset=&SelectVectorFromArray(
     Direct Array Index=$Particles.PartitionParticles.Partition
     Vector Selection Array=DataInterface() {propertyValues {FloatData [<0 0 0>, <1 1 1>, <0 0 0>, <0 0 0>, <0 0 0>, <0 0 0>, <0 0 0>, <0 0 0>, <0 0 0>, <0 0 0>, <10 0 0>, <0 -7 0>, <0 7 0>]}}
    )
    Skeletal Mesh=DataInterface() {propertyValues {
      DefaultMesh=null
      ExcludeBoneName=root
      OptionalComponentClass=null
      PreviewMesh=&F_MED_Default_Soldier
      SkinningMode=SkinOnTheFly
      SoftSourceActor=null
      SourceMode=Default
      UvSetIndex=0
      WholeMeshLOD=-1
      bExcludeBone=false
      bReadDeformedGeometry=true
      bRequireCurrentFrameData=false
      ComponentTags []
      FilteredBones [upperarm_l,upperarm_r,lowerarm_l,lowerarm_r,hand_l,hand_r,thigh_l,thigh_r,calf_l,calf_r,head,foot_r,foot_l]
      FilteredSockets []
      MeshUserParameter {parameter {name=None, typeDefHandle {registeredTypeIndex=1}, varData []}}
      SamplingRegions []
     }}
   ) {module=SkeletalMeshLocation001}
   &SkeletalMeshLocation(
    Bone / Socket Index=$Particles.PartitionParticles.Partition
    Bone Sampling Mode=::NewEnumerator4
    MaskSpawnGroup=false
    Orientation Sampling=::NewEnumerator1
    Sampled Position Offset=&SelectVectorFromArray(
     Direct Array Index=$Particles.PartitionParticles.Partition
     Vector Selection Array=DataInterface() {propertyValues {FloatData [<0 0 0>, <1 1 1>, <0 0 0>, <0 0 0>, <0 0 0>, <0 0 0>, <0 0 0>, <0 0 0>, <0 0 0>, <0 0 0>, <10 0 0>, <0 -7 0>, <0 7 0>]}}
    )
    Skeletal Mesh=DataInterface() {propertyValues {
      DefaultMesh=null
      ExcludeBoneName=root
      OptionalComponentClass=null
      PreviewMesh=&F_MED_Default_Soldier
      SkinningMode=SkinOnTheFly
      SoftSourceActor=null
      SourceMode=Default
      UvSetIndex=0
      WholeMeshLOD=-1
      bExcludeBone=false
      bReadDeformedGeometry=true
      bRequireCurrentFrameData=false
      ComponentTags []
      FilteredBones [upperarm_l,upperarm_r,lowerarm_l,lowerarm_r,hand_l,hand_r,thigh_l,thigh_r,calf_l,calf_r,head]
      FilteredSockets []
      MeshUserParameter {parameter {name=None, typeDefHandle {registeredTypeIndex=1}, varData []}}
      SamplingRegions []
     }}
   ) {enabled=false}
   &ScaleColor(
    Scale Alpha=&FloatFromCurve(FloatCurve=DataInterface() {propertyValues {Curve {keys [
         {arriveTangent=21.7774, interpMode=RCIM_Cubic, leaveTangent=21.7774, tangentMode=RCTM_User, time=0, value=0}
         {arriveTangent=-0.2619, interpMode=RCIM_Cubic, leaveTangent=-0.2619, tangentMode=RCTM_User, time=0.3, value=3}
         {arriveTangent=-5.2661, interpMode=RCIM_Cubic, leaveTangent=-5.2661, tangentMode=RCTM_User, time=1, value=0}
        ]}}})
    Scale RGB=&Multiply_VectorByFloat(
     Vector=<0.6 0.6 0.6>
     Float=&FloatFromCurve(
      FloatCurve=DataInterface() {propertyValues {Curve {keys [
          {arriveTangent=21.7774, interpMode=RCIM_Cubic, leaveTangent=21.7774, tangentMode=RCTM_User, time=0, value=0}
          {arriveTangent=-0.2619, interpMode=RCIM_Cubic, leaveTangent=-0.2619, tangentMode=RCTM_User, time=0.3, value=1.625}
          {arriveTangent=-5.2661, interpMode=RCIM_Cubic, leaveTangent=-5.2661, tangentMode=RCTM_User, time=1, value=0}
         ]}}}
      Scale Curve=&FNTimeOfDayScalar(Night Scale=0.25, Morning Scale=0.5, Evening Scale=0.5)
     )
    )
    ScaleA=false
    ScaleRGB=true
   )
   &DynamicMaterialParameters(
    Index 0 Param 1=&FloatFromCurve(
     FloatCurve=DataInterface() {propertyValues {Curve {
        defaultValue=0
        keys [
         {tangentMode=RCTM_User, time=0, value=0.55}
         {arriveTangent=-0.0405, interpMode=RCIM_Cubic, leaveTangent=-0.0405, tangentMode=RCTM_User, time=0.2, value=0.55}
         {arriveTangent=-8.6221, interpMode=RCIM_Cubic, leaveTangent=-8.6221, tangentMode=RCTM_User, time=0.3, value=0}
        ]
       }}}
     CurveIndex=$Emitter.Age
     Scale Curve=&FNTimeOfDayScalar(Night Scale=40, Morning Scale=40, Day Scale=20, Evening Scale=40)
    )
    Index 0 Param 2=&FloatFromCurve(FloatCurve=DataInterface() {propertyValues {Curve {keys [<at 0 1>,<at 1 0>]}}}, Scale Curve=0.5)
    Module.Index0Float0Write=true
    Module.Index0Float1Write=true
    Module.Index0Float2Write=false
    Module.Index0Float3Write=false
   )
  ]
  rendererClasses [&NiagaraSpriteRendererProperties, &NiagaraMeshRendererProperties]
  renderers [
   {rendererClass=&NiagaraSpriteRendererProperties}
   {rendererClass=&NiagaraMeshRendererProperties, rendererIndex=1}
  ]
 }
 Glow {
  bEnabled=false
  emitterUpdateScript [
   &EmitterState(Inactive Response=::NewEnumerator1, Life Cycle Mode=::NewEnumerator1, Loop Behavior=::NewEnumerator1, Loop Duration=0.4, UseLoopDelay=false)
   /SetVariables(Emitter.Thickness:NiagaraFloat=&FloatFromCurve(
     FloatCurve=DataInterface() {propertyValues {Curve {keys [
         {arriveTangent=-31.4678, interpMode=RCIM_Cubic, time=0, value=0.2}
         {arriveTangent=-92.2531, interpMode=RCIM_Cubic, leaveTangent=-92.2531, tangentMode=RCTM_User, time=0.0564, value=1.6547}
         {arriveTangent=-0.2405, interpMode=RCIM_Cubic, leaveTangent=-0.2405, tangentMode=RCTM_User, time=0.135, value=0.2506}
         {arriveTangent=-47.3839, interpMode=RCIM_Cubic, leaveTangent=-47.3839, tangentMode=RCTM_User, time=0.2157, value=11.1312}
         <at 0.3 0.1 RCIM_Cubic>
        ]}}}
     CurveIndex=$Emitter.Age
    )) {isSetParameters=true}
   &SpawnBurst_Instantaneous(Use Spawn Probability=false, UseLoopCountLimit=false)
  ]
  particleSpawnScript [/Niagara/Modules/Spawn/Initialization/V2/InitializeParticle(
    Color=<rgba 0 1.0841 3 0.2>
    Lifetime=0.3
    Mesh Scale=&Multiply_VectorByFloat(Vector=<1.8 0.9 1>, Float=1.25)
    Mesh Scale Mode=::NewEnumerator4
    Sprite Size=<118.0405 198.6727>
    Sprite Size Mode=::NewEnumerator4
    UsePositionOffset=false
    Write Material Random=true
   )]
  particleUpdateScript [
   &ParticleState()
   &SkeletalMeshLocation(Bone Sampling Mode=::NewEnumerator4, MaskSpawnGroup=false, Skeletal Mesh=$System.TargetSM)
   &CameraOffset(Camera Offset Amount=35)
   &ScaleColor(
    Scale Alpha=&FloatFromCurve(
     FloatCurve=DataInterface() {propertyValues {Curve {
        defaultValue=0
        keys [
         {arriveTangent=9.0725, interpMode=RCIM_Cubic, leaveTangent=9.0725, tangentMode=RCTM_User, time=-0.0042, value=0.1}
         {arriveTangent=-2.5249, interpMode=RCIM_Cubic, leaveTangent=-2.5249, tangentMode=RCTM_User, time=0.0676, value=0.7}
         {arriveTangent=-3.0116, interpMode=RCIM_Cubic, leaveTangent=-3.0116, time=0.1608, value=0.2}
         {arriveTangent=-0.0405, interpMode=RCIM_Cubic, leaveTangent=-0.0405, tangentMode=RCTM_User, time=0.3, value=0}
        ]
       }}}
     CurveIndex=$Emitter.Age
     Scale Curve=0.5
    )
    ScaleA=true
    ScaleRGB=true
   )
   &DynamicMaterialParameters(
    Index 0 Param 1=&FloatFromCurve(
     FloatCurve=DataInterface() {propertyValues {Curve {
        defaultValue=0
        keys [
         {tangentMode=RCTM_User, time=0, value=0.55}
         {arriveTangent=-0.0405, interpMode=RCIM_Cubic, leaveTangent=-0.0405, tangentMode=RCTM_User, time=0.2, value=0.55}
         {arriveTangent=-8.6221, interpMode=RCIM_Cubic, leaveTangent=-8.6221, tangentMode=RCTM_User, time=0.3, value=0}
        ]
       }}}
     CurveIndex=$Emitter.Age
     Scale Curve=&FNTimeOfDayScalar(Night Scale=40, Morning Scale=40, Day Scale=20, Evening Scale=40)
    )
    Index 0 Param 2=&FloatFromCurve(FloatCurve=DataInterface() {propertyValues {Curve {keys [<at 0 1>,<at 1 0>]}}}, Scale Curve=0.5)
    Module.Index0Float0Write=true
    Module.Index0Float1Write=true
    Module.Index0Float2Write=false
    Module.Index0Float3Write=false
   )
  ]
  rendererClasses [&NiagaraMeshRendererProperties]
  renderers [{rendererClass=&NiagaraMeshRendererProperties, rendererIndex=0}]
 }
 LW_Small_Chunks {rendererClasses [], renderers []}
}
summary {
 systemName=NS_Figment_NoDamage_Hit
 emitters [
  {bEnabled=false, emitterName=Glow, rendererClasses [&NiagaraMeshRendererProperties]}
  {emitterName=LW_Small_Chunks, rendererClasses []}
  {
   bEnabled=false
   emitterName=Bright_Flash_Decal
   rendererClasses [&NiagaraSpriteRendererProperties, &NiagaraMeshRendererProperties]
  }
  {
   emitterName=Bright_Flash_Decal001
   rendererClasses [&NiagaraSpriteRendererProperties, &NiagaraMeshRendererProperties]
  }
 ]
 userVariables [
  User.Ammo {type=NiagaraFloat, defaultValue=1, description=""}
  User.Ascended {type=NiagaraBool, defaultValue=true}
  User.bIsFrontEnd {type=NiagaraBool}
  User.bIsFrontEndPreview {type=NiagaraBool}
  User.IdleGlow {type=NiagaraFloat, defaultValue=0, description=""}
  User.SkeletalMesh {
   type=NiagaraDataInterfaceSkeletalMesh
   defaultValue {
    struct=NiagaraExt_VariableValue_DataInterface
    value {dataInterface=&self:NiagaraDataInterfaceSkeletalMesh_0}
   }
  }
  User.StartFrontEndLoop {type=NiagaraBool, defaultValue=true}
  User.Variant {type=NiagaraFloat, defaultValue=0, description=""}
  User.Variant_Float {type=NiagaraFloat, defaultValue=0, description=""}
 ]
}
